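# MetaSound source for the bomb explosion foley in the SouthBalk emote, played as a stereo one-shot.
# It is a preset of the shared pickaxe impact MetaSound, supplying four explosion wave variations plus random pitch and volume ranges with no start delay.
# Emote foley sound class, attenuation, property sheet and concurrency settings govern mixing and spatial playback, and it keeps playing when virtualized.
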MetaSoundSource /BRCosmetics/Sounds/Emotes/SouthBalk/MSS_Emote_Foley_SouthBalk_Bomb_Explode_PitchRando
preset_parent=/FortMetasound/Cosmetics/Pickaxe/MSS_Pickaxe_Impact
inputs [
 Impact.Sounds {type=WaveAsset:Array}
 PitchModulationRange.Max {type=Float}
 PitchModulationRange.Min {type=Float}
 StartDelay.Max {type=Time}
 StartDelay.Min {type=Time}
 TemplatePitchMultiplier {type=Float}
 TemplateVolumeMultiplier {type=Float}
 UE.Source.OnPlay {type=Trigger}
 VolumeModulationRange.Max {type=Float}
 VolumeModulationRange.Min {type=Float}
]
interfaces [UE.OutputFormat.Stereo:1.0, UE.Source.OneShot:1.0, UE.Source:1.1]
outputs [
 "UE.OutputFormat.Stereo.Audio:0" {type=Audio}
 "UE.OutputFormat.Stereo.Audio:1" {type=Audio}
 UE.Source.OneShot.OnFinished {type=Trigger}
]
pages [Default {input_defaults {
   PitchModulationRange.Max=1.1325
   PitchModulationRange.Min=0.874
   StartDelay.Max=0
   StartDelay.Min=0
   TemplatePitchMultiplier=1
   TemplateVolumeMultiplier=1
   UE.Source.OnPlay=null
   VolumeModulationRange.Max=1.0479
   VolumeModulationRange.Min=0.7208
   Impact.Sounds [
    &/Emote_SouthBalk_Bomb_Explode_01
    &/Emote_SouthBalk_Bomb_Explode_02
    &/Emote_SouthBalk_Bomb_Explode_03
    &/Emote_SouthBalk_Bomb_Explode_04
   ]
  }}]
properties {
 assetImportData=&self:AssetImportData
 attenuationSettings=&Fort_Emotes_Athena_Att
 audioPropertiesSheet=/BRCosmetics/Sounds/Settings/PropertySheets/Emotes/APS_Cosmetic_Emote_Foley
 maxDistance=1600
 numChannels=2
 outputFormat=Stereo
 sampleRate=48000
 soundClassObject=/Game/Sounds/FortSoundClassesAndModes/Fort_Foley_Emote
 virtualizationMode=PlayWhenSilent
 volume=0.7
 concurrencySet [/Game/Sounds/Concurrency_Rules/Athena_EmoteFoley_Conc]
}
